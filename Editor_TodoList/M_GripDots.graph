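# M_GripDots is a masked UI-domain material in Editor_TodoList, apparently for drawing a grip-dots handle graphic.
# It contains a sample of the T_GripDots texture and a Colour vector parameter defaulting to opaque light grey.
# The graph lists no links from either node to each other or to the material outputs.
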Material /Editor_TodoList/M_GripDots
graph {
 properties {bAutomaticallySetUsageInEditor=false, blendMode=BLEND_Masked, materialDomain=MD_UI, usedShadingModels=Default Lit}
 nodes [
  TextureSample() {texture=/Editor_TodoList/Textures/T_GripDots}
  VectorParameter() {parameterName=Colour, defaultValue=<rgba 0.6875 0.6875 0.6875 1>}
 ]
}
