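# Editor Utility Widget for one to-do list row, with checkbox, editable message, priority and delete buttons, and an unwired reorder grip.
# Construct fills it from Todo Data, showing one of two text boxes by completion and styling priority, then sets Ready.
# Edits call Modify Entry or Remove Entry on the Parent list, with checkbox and text events gated by Ready.
# Modify calls keep the GUID but clear tooltip and Jira fields.
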EditorUtilityWidgetBlueprint /Editor_TodoList/EUW_TodoList_Entry
graphs [
 EventGraph {nodes [
   n1 = "On Check State Changed (CB_CompletedTodo)"!() -> (OutputDelegate, then, bIsChecked:bool) {is=title K2Node_ComponentBoundEvent}
   > Branch(Condition=Get Ready() -> (Ready:bool))
   n2 = Get!() -> (
    Todo Data_Completed_1_#1:bool=~
    Todo Data_Message_4_#2:Text
    Todo Data_Tooltip_19_#3:Text
    Todo Data_GUID_7_#4:Guid Structure
    Todo Data_PriorityTask_9_#5:bool=~
    Todo Data_JiraTask_15_#6:bool=~
    Todo Data_JiraAddress_12_#7:String
   )
   > Modify Entry!(
    execute
    self:To-do List Object Reference=Get Parent() -> (Parent:To-do List Object Reference)
    Updated Data:Struct Todo List Data Structure=Make Struct Todo List Data!(
     Completed_1_#1:bool=@n1.bIsChecked
     Message_4_#2:Text=@n2.Todo Data_Message_4_#2
     Tooltip_19_#3:Text
     GUID_7_#4:Guid Structure=@n2.Todo Data_GUID_7_#4
     PriorityTask_9_#5:bool=@n2.Todo Data_PriorityTask_9_#5
     JiraTask_15_#6:bool=False
     JiraAddress_12_#7:String
    ) -> (Struct_TodoListData:Struct Todo List Data Structure) {is=title K2Node_MakeStruct}
   ) -> (then)
   n3 = "OnTextCommitted (Multi-Line Editable Text) (MessageNormal)"!() -> (OutputDelegate, then, Text:"Text (by ref)", CommitMethod:ETextCommit Enum=Default) {is=title K2Node_ComponentBoundEvent}
   n4 = Get Ready()
   > Branch(Condition=@n4)
   n5 = IsChecked(self=Get CB_CompletedTodo() -> (CB_CompletedTodo:Editor Utility Check Box Object Reference))
   n6 = Get Parent()
   n7 = Get() -> (
    Todo Data_Completed_1_#1=~
    Todo Data_PriorityTask_9_#5=~
    Todo Data_JiraTask_15_#6=~
   )
   > Modify Entry!(
    execute
    self=@n6
    Updated Data_Completed_1_#1:bool=@n5
    Updated Data_Message_4_#2:Text=@n3.Text
    Updated Data_Tooltip_19_#3:Text
    Updated Data_GUID_7_#4:Guid Structure=@n7.Todo Data_GUID_7_#4
    Updated Data_PriorityTask_9_#5:bool=@n7.Todo Data_PriorityTask_9_#5
    Updated Data_JiraTask_15_#6:bool=False
    Updated Data_JiraAddress_12_#7:String
   ) -> (then)
   n8 = "OnTextCommitted (Multi-Line Editable Text) (MessageCompleted)"!() -> (OutputDelegate, then, Text:"Text (by ref)", CommitMethod:ETextCommit Enum=Default) {is=title K2Node_ComponentBoundEvent}
   > Branch(Condition=@n4)
   > Modify Entry(
    self=@n6
    Updated Data_Completed_1_#1=@n5
    Updated Data_Message_4_#2=@n8.Text
    Updated Data_GUID_7_#4=@n7.Todo Data_GUID_7_#4
    Updated Data_PriorityTask_9_#5=@n7.Todo Data_PriorityTask_9_#5
    Updated Data_JiraTask_15_#6=False
   )
   "On Clicked (Btn_Delete)"() {is=title K2Node_ComponentBoundEvent}
   > Remove Entry!(
    execute
    self:To-do List Object Reference=Get Parent()
    GUID:Guid Structure=Get() -> (
     Todo Data_Completed_1_#1=~
     Todo Data_PriorityTask_9_#5=~
     Todo Data_JiraTask_15_#6=~
    ).Todo Data_GUID_7_#4
   ) -> (then)
   "On Clicked (Btn_Priority)"() {is=title K2Node_ComponentBoundEvent}
   n9 = Get() -> (
    Todo Data_Completed_1_#1=~
    Todo Data_PriorityTask_9_#5=~
    Todo Data_JiraTask_15_#6=~
   )
   > Modify Entry(
    self=Get Parent()
    Updated Data_Completed_1_#1=@n9
    Updated Data_Message_4_#2=@n9.Todo Data_Message_4_#2
    Updated Data_GUID_7_#4=@n9.Todo Data_GUID_7_#4
    Updated Data_PriorityTask_9_#5=NOT Boolean(A=@n9.Todo Data_PriorityTask_9_#5)
    Updated Data_JiraTask_15_#6=False
   )
   Event Construct()
   > n10 = Sequence()
   n11 = Get() -> (
    Todo Data_Completed_1_#1=~
    Todo Data_PriorityTask_9_#5=~
    Todo Data_JiraTask_15_#6=~
   )
   > "SetText (Multi-Line Editable Text)"!(
    execute
    self:"Editable Text (Multi- Line) Object Reference"=[
     Get MessageCompleted() -> (MessageCompleted:Editor Utility Multi Line Editable Text Object Reference)
     Get MessageNormal() -> (MessageNormal:Editor Utility Multi Line Editable Text Object Reference)
    ]
    InText:Text=@n11.Todo Data_Message_4_#2
   ) -> (then)
   > K2Node_Composite_0 = Set Fonts!(execute) -> (then:Exec) {is=title K2Node_Composite}
   > SetCheckedState!(
    execute
    self:Check Box Object Reference=Get CB_CompletedTodo()
    InCheckedState:ECheckBoxState Enum=Select(Option 0:ECheckBoxState Enum=Unchecked, Option 1:ECheckBoxState Enum=Checked, Index=@n11) -> (ReturnValue:ECheckBoxState Enum)
   ) -> (then)
   > n12 = Branch(Condition=@n11)
   n13 = Get MessageCompleted()
   > SetVisibility(self=@n13, InVisibility=Visible)
   n14 = Get MessageNormal()
   > SetVisibility(self=@n14, InVisibility=Collapsed)
   SetVisibility(execute=@n12.else, self=@n13, InVisibility=Collapsed)
   > SetVisibility(self=@n14, InVisibility=Visible)
   Set Ready(execute=@n10.then_1, Ready:bool=true) -> (Output_Get:bool)
   Event PreConstruct()
  ]}
 Set Fonts {
  path=EventGraph.K2Node_Composite_0.Set Fonts
  nodes [
   Inputs()
   > n1 = Sequence!(execute) -> (then_0, then_1, then_2:Exec)
   n2 = Get() -> (
    Todo Data_Completed_1_#1=~
    Todo Data_PriorityTask_9_#5=~
    Todo Data_JiraTask_15_#6=~
   )
   n3 = Get MessageCompleted()
   > n4 = Set members in Slate Font Info!(
    execute:Exec
    StructRef:"Slate Font Info Structure (by ref)"=GetFont!(self:"Editable Text (Multi- Line) Object Reference"=@n3) -> (ReturnValue:"Slate Font Info Structure (by ref)")
    TypefaceFontName:Name=SelectName!(A:Name=NotoSans-Bold, B:Name=NotoSans-Regular, bPickA:bool=@n2.Todo Data_PriorityTask_9_#5) -> (ReturnValue:Name=None)
   ) -> (then:Exec, StructOut:"Slate Font Info Structure (by ref)") {is=title K2Node_SetFieldsInStruct}
   > SetFont(self:"Editable Text (Multi- Line) Object Reference"=@n3, InFontInfo=@n4)
   n5 = Get MessageNormal()
   n6 = Set members in Slate Font Info(
    execute=@n1.then_1
    StructRef=GetFont(self=@n5)
    TypefaceFontName=SelectName(A=Extra Bold, B=Regular, bPickA=@n2.Todo Data_PriorityTask_9_#5) -> (ReturnValue=None)
   )
   > SetFont(self=@n5, InFontInfo=@n6)
   SetBackgroundColor!(
    execute=@n1.then_2
    self:Button Object Reference=Get Btn_Priority() -> (Btn_Priority:Editor Utility Button Object Reference)
    InBackgroundColor:Linear Color Structure=Select(
     Option 0:Linear Color Structure=(R=0.3647,G=0.3765,B=0.3373,A=1)
     Option 1:Linear Color Structure=(R=0.4844,G=0.1483,B=0.0854,A=1)
     Index=@n2.Todo Data_PriorityTask_9_#5
    ) -> (ReturnValue:Linear Color Structure)
   ) -> (then)
   > Outputs!(then:Exec)
  ]
 }
]
variables [
 Completed {type=boolean}
 Parent {type=EUW_TodoList_C, default=null}
 Ready {type=boolean}
 Todo Data {
  type=Struct_TodoListData
  default {
   completed=false
   gUId=#8
   jira Address=""
   jira Task=false
   message=Test Message
   priority Task=false
   tooltip=""
  }
 }
]
widgets {
 info {rootWidgetClass=&StackBox, widgetCount=9}
 widgets [
  StackBox_1 {widgetClassPath=&StackBox, widget=StackBox_1, properties {orientation=Orient_Vertical}}
  StackBox_0 {widgetClassPath=&StackBox, parent=StackBox_1, widget=StackBox_0, slot {class=StackBoxSlot}}
  Btn_GripDots {
   widgetClassPath=&EditorUtilityButton
   parent=StackBox_0
   widget=Btn_GripDots
   properties {
    bOverride_Cursor=true
    cursor=GrabHand
    toolTipText=Reorder Entry
    widgetStyle {
     clickedSlateSound {resourceObject=null}
     disabled {
      bIsDynamicallyLoaded=false
      drawAs=RoundedBox
      imageType=NoImage
      mirroring=NoMirror
      resourceName=None
      resourceObject=null
      tiling=NoTile
      imageSize=<0 0>
      margin=<ltrb 0 0 0 0>
      outlineSettings {
       bUseBrushTransparency=false
       roundingType=FixedRadius
       width=1
       color {colorUseRule=UseColor_ColorTable, specifiedColor=<rgba 0 0 0 0>}
       cornerRadii=<4 4 4 4>
      }
      tintColor {colorUseRule=UseColor_Specified, specifiedColor=<rgba 0.0395 0.0395 0.0395 1>}
      uVRegion {bIsValid=false, max=<0 0>, min=<0 0>}
     }
     disabledForeground {colorUseRule=UseColor_ColorTable, specifiedColor=<rgba 0 0 0 0>}
     hovered {
      bIsDynamicallyLoaded=false
      drawAs=Image
      imageType=FullColor
      mirroring=NoMirror
      resourceName=None
      resourceObject=&/M_GripDots
      tiling=NoTile
      imageSize=<12 12>
      margin=<ltrb 0 0 0 0>
      outlineSettings {
       bUseBrushTransparency=false
       roundingType=FixedRadius
       width=1
       color {colorUseRule=UseColor_ColorTable, specifiedColor=<rgba 0 0 0 0>}
       cornerRadii=<4 4 4 4>
      }
      tintColor {colorUseRule=UseColor_Specified, specifiedColor=<rgba 0.3526 0.5 0.4792 1>}
      uVRegion {bIsValid=false, max=<0 0>, min=<0 0>}
     }
     hoveredForeground {colorUseRule=UseColor_ColorTable, specifiedColor=<rgba 0 0 0 0>}
     hoveredSlateSound {resourceObject=null}
     normal {
      bIsDynamicallyLoaded=false
      drawAs=Image
      imageType=FullColor
      mirroring=NoMirror
      resourceName=None
      resourceObject=&M_GripDots
      tiling=NoTile
      imageSize=<12 12>
      margin=<ltrb 0 0 0 0>
      outlineSettings {
       bUseBrushTransparency=false
       roundingType=FixedRadius
       width=1
       color {colorUseRule=UseColor_ColorTable, specifiedColor=<rgba 0 0 0 0>}
       cornerRadii=<4 4 4 4>
      }
      tintColor {colorUseRule=UseColor_Specified, specifiedColor=<rgba 0.1433 0.2031 0.1947 1>}
      uVRegion {bIsValid=false, max=<0 0>, min=<0 0>}
     }
     normalForeground {colorUseRule=UseColor_ColorTable, specifiedColor=<rgba 0 0 0 0>}
     normalPadding=<ltrb 12 1.5 12 1.5>
     pressed {
      bIsDynamicallyLoaded=false
      drawAs=Image
      imageType=FullColor
      mirroring=NoMirror
      resourceName=None
      resourceObject=&M_GripDots
      tiling=NoTile
      imageSize=<12 12>
      margin=<ltrb 0 0 0 0>
      outlineSettings {
       bUseBrushTransparency=false
       roundingType=FixedRadius
       width=1
       color {colorUseRule=UseColor_ColorTable, specifiedColor=<rgba 0 0 0 0>}
       cornerRadii=<4 4 4 4>
      }
      tintColor {colorUseRule=UseColor_Specified, specifiedColor=<rgba 0.4953 0.7763 0.7865 1>}
      uVRegion {bIsValid=false, max=<0 0>, min=<0 0>}
     }
     pressedForeground {colorUseRule=UseColor_ColorTable, specifiedColor=<rgba 0 0 0 0>}
     pressedPadding=<ltrb 12 2.5 12 0.5>
     pressedSlateSound {resourceObject=null}
    }
   }
   slot {class=StackBoxSlot, properties {horizontalAlignment=HAlign_Center, verticalAlignment=VAlign_Center, padding=<ltrb 2 2 2 2>}}
  }
  CB_CompletedTodo {
   widgetClassPath=&EditorUtilityCheckBox
   bIsVariable=true
   parent=StackBox_0
   widget=CB_CompletedTodo
   properties {toolTipText=Mark Entry Complete}
   slot {class=StackBoxSlot, properties {horizontalAlignment=HAlign_Left, verticalAlignment=VAlign_Center, padding=<ltrb 0 5 10 5>}}
  }
  MessageNormal {
   widgetClassPath=/Script/Blutility.EditorUtilityMultiLineEditableText
   bIsVariable=true
   parent=StackBox_0
   widget=MessageNormal
   properties {
    text=Example Text here
    widgetStyle {
     overflowPolicy=Clip
     transformPolicy=None
     colorAndOpacity {colorUseRule=UseColor_Specified, specifiedColor=<rgba 0.9115 0.9115 0.9115 1>}
     font {
      bForceMonospaced=false
      bMaterialIsStencil=false
      fontMaterial=null
      fontObject=/LauncherSocial/BrutalType
      letterSpacing=0
      monospacedWidth=1
      size=9.38
      skewAmount=0
      typefaceFontName=Regular
      outlineSettings {
       bApplyOutlineToDropShadows=false
       bMiteredCorners=false
       bSeparateFillAlpha=false
       outlineMaterial=null
       outlineSize=0
       outlineColor=<rgba 0 0 0 1>
      }
     }
     highlightColor {colorUseRule=UseColor_ColorTable, specifiedColor=<rgba 0 0 0 0>}
     highlightShape {
      bIsDynamicallyLoaded=false
      drawAs=Image
      imageType=NoImage
      mirroring=NoMirror
      resourceName=None
      resourceObject=null
      tiling=NoTile
      imageSize=<0 0>
      margin=<ltrb 0 0 0 0>
      outlineSettings {
       bUseBrushTransparency=false
       roundingType=HalfHeightRadius
       width=0
       color {colorUseRule=UseColor_Specified, specifiedColor=<rgba 0 0 0 0>}
       cornerRadii=<0 0 0 0>
      }
      tintColor {colorUseRule=UseColor_Specified, specifiedColor=<rgba 0.2582 0.5395 0.0685 1>}
      uVRegion {bIsValid=false, max=<0 0>, min=<0 0>}
     }
     selectedBackgroundColor {colorUseRule=UseColor_Specified, specifiedColor=<rgba 0 0.162 0.7454 1>}
     shadowColorAndOpacity=<rgba 0 0 0 1>
     shadowOffset=<0 0>
     strikeBrush {
      bIsDynamicallyLoaded=false
      drawAs=Image
      imageType=NoImage
      mirroring=NoMirror
      resourceName=None
      resourceObject=null
      tiling=NoTile
      imageSize=<32 32>
      margin=<ltrb 0 0 0 0>
      outlineSettings {
       bUseBrushTransparency=false
       roundingType=HalfHeightRadius
       width=0
       color {colorUseRule=UseColor_Specified, specifiedColor=<rgba 0 0 0 0>}
       cornerRadii=<0 0 0 0>
      }
      tintColor {colorUseRule=UseColor_Specified, specifiedColor=<rgba 1 1 1 1>}
      uVRegion {bIsValid=false, max=<0 0>, min=<0 0>}
     }
     underlineBrush {
      bIsDynamicallyLoaded=false
      drawAs=Image
      imageType=NoImage
      mirroring=NoMirror
      resourceName=None
      resourceObject=null
      tiling=NoTile
      imageSize=<32 32>
      margin=<ltrb 0 0 0 0>
      outlineSettings {
       bUseBrushTransparency=false
       roundingType=HalfHeightRadius
       width=0
       color {colorUseRule=UseColor_Specified, specifiedColor=<rgba 0 0 0 0>}
       cornerRadii=<0 0 0 0>
      }
      tintColor {colorUseRule=UseColor_Specified, specifiedColor=<rgba 1 1 1 1>}
      uVRegion {bIsValid=false, max=<0 0>, min=<0 0>}
     }
    }
   }
   slot {class=StackBoxSlot, properties {verticalAlignment=VAlign_Center, size {sizeRule=Fill, value=1}}}
  }
  MessageCompleted {
   widgetClassPath=&EditorUtilityMultiLineEditableText
   bIsVariable=true
   parent=StackBox_0
   widget=MessageCompleted
   properties {
    text=Example Text here
    widgetStyle {
     overflowPolicy=Clip
     transformPolicy=None
     colorAndOpacity {colorUseRule=UseColor_Specified, specifiedColor=<rgba 0.7031 0.7031 0.7031 0.4>}
     font {
      bForceMonospaced=false
      bMaterialIsStencil=false
      fontMaterial=null
      fontObject=&NotoSans
      letterSpacing=0
      monospacedWidth=1
      size=9
      skewAmount=0.4
      typefaceFontName=NotoSans-Regular
      outlineSettings {
       bApplyOutlineToDropShadows=false
       bMiteredCorners=false
       bSeparateFillAlpha=false
       outlineMaterial=null
       outlineSize=0
       outlineColor=<rgba 0 0 0 1>
      }
     }
     highlightColor {colorUseRule=UseColor_ColorTable, specifiedColor=<rgba 0 0 0 0>}
     highlightShape {
      bIsDynamicallyLoaded=false
      drawAs=Image
      imageType=NoImage
      mirroring=NoMirror
      resourceName=None
      resourceObject=null
      tiling=NoTile
      imageSize=<0 0>
      margin=<ltrb 0 0 0 0>
      outlineSettings {
       bUseBrushTransparency=false
       roundingType=HalfHeightRadius
       width=0
       color {colorUseRule=UseColor_Specified, specifiedColor=<rgba 0 0 0 0>}
       cornerRadii=<0 0 0 0>
      }
      tintColor {colorUseRule=UseColor_Specified, specifiedColor=<rgba 0.2582 0.5395 0.0685 1>}
      uVRegion {bIsValid=false, max=<0 0>, min=<0 0>}
     }
     selectedBackgroundColor {colorUseRule=UseColor_Specified, specifiedColor=<rgba 0 0.162 0.7454 1>}
     shadowColorAndOpacity=<rgba 0 0 0 1>
     shadowOffset=<0 0>
     strikeBrush {
      bIsDynamicallyLoaded=false
      drawAs=Image
      imageType=NoImage
      mirroring=NoMirror
      resourceName=None
      resourceObject=null
      tiling=NoTile
      imageSize=<32 32>
      margin=<ltrb 0 0 0 0>
      outlineSettings {
       bUseBrushTransparency=false
       roundingType=HalfHeightRadius
       width=0
       color {colorUseRule=UseColor_Specified, specifiedColor=<rgba 0 0 0 0>}
       cornerRadii=<0 0 0 0>
      }
      tintColor {colorUseRule=UseColor_Specified, specifiedColor=<rgba 1 1 1 1>}
      uVRegion {bIsValid=false, max=<0 0>, min=<0 0>}
     }
     underlineBrush {
      bIsDynamicallyLoaded=false
      drawAs=Image
      imageType=NoImage
      mirroring=NoMirror
      resourceName=None
      resourceObject=null
      tiling=NoTile
      imageSize=<32 32>
      margin=<ltrb 0 0 0 0>
      outlineSettings {
       bUseBrushTransparency=false
       roundingType=HalfHeightRadius
       width=0
       color {colorUseRule=UseColor_Specified, specifiedColor=<rgba 0 0 0 0>}
       cornerRadii=<0 0 0 0>
      }
      tintColor {colorUseRule=UseColor_Specified, specifiedColor=<rgba 1 1 1 1>}
      uVRegion {bIsValid=false, max=<0 0>, min=<0 0>}
     }
    }
   }
   slot {class=StackBoxSlot, properties {verticalAlignment=VAlign_Center, size {sizeRule=Fill, value=1}}}
  }
  Btn_Priority {
   widgetClassPath=&EditorUtilityButton
   parent=StackBox_0
   widget=Btn_Priority
   properties {
    bOverride_Cursor=true
    cursor=Hand
    toolTipText=Set Priority Entry
    widgetStyle {
     clickedSlateSound {resourceObject=null}
     disabled {
      bIsDynamicallyLoaded=false
      drawAs=RoundedBox
      imageType=NoImage
      mirroring=NoMirror
      resourceName=None
      resourceObject=null
      tiling=NoTile
      imageSize=<0 0>
      margin=<ltrb 0 0 0 0>
      outlineSettings {
       bUseBrushTransparency=false
       roundingType=FixedRadius
       width=1
       color {colorUseRule=UseColor_ColorTable, specifiedColor=<rgba 0 0 0 0>}
       cornerRadii=<4 4 4 4>
      }
      tintColor {colorUseRule=UseColor_Specified, specifiedColor=<rgba 0.0395 0.0395 0.0395 1>}
      uVRegion {bIsValid=false, max=<0 0>, min=<0 0>}
     }
     disabledForeground {colorUseRule=UseColor_ColorTable, specifiedColor=<rgba 0 0 0 0>}
     hovered {
      bIsDynamicallyLoaded=false
      drawAs=Image
      imageType=FullColor
      mirroring=NoMirror
      resourceName=None
      resourceObject=/Editor_TodoList/Textures/T_Todo_Important
      tiling=NoTile
      imageSize=<16 16>
      margin=<ltrb 0 0 0 0>
      outlineSettings {
       bUseBrushTransparency=false
       roundingType=FixedRadius
       width=1
       color {colorUseRule=UseColor_ColorTable, specifiedColor=<rgba 0 0 0 0>}
       cornerRadii=<4 4 4 4>
      }
      tintColor {colorUseRule=UseColor_Specified, specifiedColor=<rgba 0.5468 0.5625 0.5085 1>}
      uVRegion {bIsValid=false, max=<0 0>, min=<0 0>}
     }
     hoveredForeground {colorUseRule=UseColor_ColorTable, specifiedColor=<rgba 0 0 0 0>}
     hoveredSlateSound {resourceObject=null}
     normal {
      bIsDynamicallyLoaded=false
      drawAs=Image
      imageType=FullColor
      mirroring=NoMirror
      resourceName=None
      resourceObject=&T_Todo_Important
      tiling=NoTile
      imageSize=<16 16>
      margin=<ltrb 0 0 0 0>
      outlineSettings {
       bUseBrushTransparency=false
       roundingType=FixedRadius
       width=1
       color {colorUseRule=UseColor_ColorTable, specifiedColor=<rgba 0 0 0 0>}
       cornerRadii=<4 4 4 4>
      }
      tintColor {colorUseRule=UseColor_Specified, specifiedColor=<rgba 0.3645 0.375 0.339 1>}
      uVRegion {bIsValid=false, max=<0 0>, min=<0 0>}
     }
     normalForeground {colorUseRule=UseColor_ColorTable, specifiedColor=<rgba 0 0 0 0>}
     normalPadding=<ltrb 12 1.5 12 1.5>
     pressed {
      bIsDynamicallyLoaded=false
      drawAs=Image
      imageType=FullColor
      mirroring=NoMirror
      resourceName=None
      resourceObject=&T_Todo_Important
      tiling=NoTile
      imageSize=<16 16>
      margin=<ltrb 0 0 0 0>
      outlineSettings {
       bUseBrushTransparency=false
       roundingType=FixedRadius
       width=1
       color {colorUseRule=UseColor_ColorTable, specifiedColor=<rgba 0 0 0 0>}
       cornerRadii=<4 4 4 4>
      }
      tintColor {colorUseRule=UseColor_Specified, specifiedColor=<rgba 0.9366 0.9635 0.871 1>}
      uVRegion {bIsValid=false, max=<0 0>, min=<0 0>}
     }
     pressedForeground {colorUseRule=UseColor_ColorTable, specifiedColor=<rgba 0 0 0 0>}
     pressedPadding=<ltrb 12 2.5 12 0.5>
     pressedSlateSound {resourceObject=null}
    }
   }
   slot {class=StackBoxSlot, properties {horizontalAlignment=HAlign_Center, verticalAlignment=VAlign_Center, padding=<ltrb 10 0 5 0>}}
  }
  Btn_Delete {
   widgetClassPath=&EditorUtilityButton
   parent=StackBox_0
   widget=Btn_Delete
   properties {
    bOverride_Cursor=true
    cursor=Hand
    toolTipText=Delete Entry
    widgetStyle {
     clickedSlateSound {resourceObject=null}
     disabled {
      bIsDynamicallyLoaded=false
      drawAs=RoundedBox
      imageType=NoImage
      mirroring=NoMirror
      resourceName=None
      resourceObject=null
      tiling=NoTile
      imageSize=<0 0>
      margin=<ltrb 0 0 0 0>
      outlineSettings {
       bUseBrushTransparency=false
       roundingType=FixedRadius
       width=1
       color {colorUseRule=UseColor_ColorTable, specifiedColor=<rgba 0 0 0 0>}
       cornerRadii=<4 4 4 4>
      }
      tintColor {colorUseRule=UseColor_Specified, specifiedColor=<rgba 0.0395 0.0395 0.0395 1>}
      uVRegion {bIsValid=false, max=<0 0>, min=<0 0>}
     }
     disabledForeground {colorUseRule=UseColor_ColorTable, specifiedColor=<rgba 0 0 0 0>}
     hovered {
      bIsDynamicallyLoaded=false
      drawAs=Image
      imageType=FullColor
      mirroring=NoMirror
      resourceName=None
      resourceObject=/Editor_TodoList/Textures/T_Todo_Delete
      tiling=NoTile
      imageSize=<16 16>
      margin=<ltrb 0 0 0 0>
      outlineSettings {
       bUseBrushTransparency=false
       roundingType=FixedRadius
       width=1
       color {colorUseRule=UseColor_ColorTable, specifiedColor=<rgba 0 0 0 0>}
       cornerRadii=<4 4 4 4>
      }
      tintColor {colorUseRule=UseColor_Specified, specifiedColor=<rgba 0.5468 0.5625 0.5085 1>}
      uVRegion {bIsValid=false, max=<0 0>, min=<0 0>}
     }
     hoveredForeground {colorUseRule=UseColor_ColorTable, specifiedColor=<rgba 0 0 0 0>}
     hoveredSlateSound {resourceObject=null}
     normal {
      bIsDynamicallyLoaded=false
      drawAs=Image
      imageType=FullColor
      mirroring=NoMirror
      resourceName=None
      resourceObject=&T_Todo_Delete
      tiling=NoTile
      imageSize=<16 16>
      margin=<ltrb 0 0 0 0>
      outlineSettings {
       bUseBrushTransparency=false
       roundingType=FixedRadius
       width=1
       color {colorUseRule=UseColor_ColorTable, specifiedColor=<rgba 0 0 0 0>}
       cornerRadii=<4 4 4 4>
      }
      tintColor {colorUseRule=UseColor_Specified, specifiedColor=<rgba 0.3645 0.375 0.339 1>}
      uVRegion {bIsValid=false, max=<0 0>, min=<0 0>}
     }
     normalForeground {colorUseRule=UseColor_ColorTable, specifiedColor=<rgba 0 0 0 0>}
     normalPadding=<ltrb 12 1.5 12 1.5>
     pressed {
      bIsDynamicallyLoaded=false
      drawAs=Image
      imageType=FullColor
      mirroring=NoMirror
      resourceName=None
      resourceObject=&T_Todo_Delete
      tiling=NoTile
      imageSize=<16 16>
      margin=<ltrb 0 0 0 0>
      outlineSettings {
       bUseBrushTransparency=false
       roundingType=FixedRadius
       width=1
       color {colorUseRule=UseColor_ColorTable, specifiedColor=<rgba 0 0 0 0>}
       cornerRadii=<4 4 4 4>
      }
      tintColor {colorUseRule=UseColor_Specified, specifiedColor=<rgba 0.9366 0.9635 0.871 1>}
      uVRegion {bIsValid=false, max=<0 0>, min=<0 0>}
     }
     pressedForeground {colorUseRule=UseColor_ColorTable, specifiedColor=<rgba 0 0 0 0>}
     pressedPadding=<ltrb 12 2.5 12 0.5>
     pressedSlateSound {resourceObject=null}
    }
   }
   slot {class=StackBoxSlot, properties {horizontalAlignment=HAlign_Right, verticalAlignment=VAlign_Center, padding=<ltrb 5 0 10 0>}}
  }
  Image_88 {
   widgetClassPath=&Image
   bIsVariable=true
   parent=StackBox_1
   widget=Image_88
   properties {brush {
     bIsDynamicallyLoaded=false
     drawAs=Image
     imageType=NoImage
     mirroring=NoMirror
     resourceName=None
     resourceObject=null
     tiling=NoTile
     imageSize=<500 1>
     margin=<ltrb 0 0 0 0>
     outlineSettings {
      bUseBrushTransparency=false
      roundingType=HalfHeightRadius
      width=0
      color {colorUseRule=UseColor_Specified, specifiedColor=<rgba 0 0 0 0>}
      cornerRadii=<0 0 0 0>
     }
     tintColor {colorUseRule=UseColor_Specified, specifiedColor=<rgba 0.1118 0.235 0.2448 0.15>}
     uVRegion {bIsValid=false, max=<0 0>, min=<0 0>}
    }}
   slot {class=StackBoxSlot, properties {horizontalAlignment=HAlign_Center, padding=<ltrb 5 5 5 5>}}
  }
 ]
}
